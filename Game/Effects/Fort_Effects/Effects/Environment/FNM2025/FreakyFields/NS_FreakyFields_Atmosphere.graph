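# Niagara atmosphere system for the FNM2025 Freaky Fields environment, with GPU bat and fog emitters both reacting to time of day.
# Mesh bats spawn in a high torus, mostly near day-cycle ends, with distance culling and a count-based kill.
# Vortex, curl noise, camera repulsion and distance-field avoidance steer them.
# Drifting sprite and mesh fog follows landscape normals, shrinks near the viewer, darkens mid-cycle and is cleared by hardcoded world-space kill boxes.
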NiagaraSystem /Game/Effects/Fort_Effects/Effects/Environment/FNM2025/FreakyFields/NS_FreakyFields_Atmosphere
dependencies {
 usedDataInterfaces [
  &NiagaraDataInterfaceCurve
  &NiagaraDataInterfaceLandscape
  &NiagaraDataInterfaceGBuffer
  &NiagaraDataInterfaceCamera
 ]
 usedDynamicInputs []
 usedModules [&SystemState]
}
emitters {
 Bats001 {
  simTarget=GPUComputeSim
  emitterSpawnScript [/SetVariables(Emitter.PositionOffset:NiagaraPosition=<0 0 7500>, Emitter.RadiusMult=1.35) {isSetParameters=true}]
  emitterUpdateScript [
   &EmitterState(
    Enable Distance Culling=true
    Life Cycle Mode=::NewEnumerator1
    Loop Behavior=::NewEnumerator1
    Loop Duration Mode=::NewEnumerator1
    MaxDistance=10000
    Min Distance Response=::SleepAndLetParticlesFinish
    Module.bApplySpawnCountScaleByDistance=true
    Module.bUseMaxDistance=true
    Module.bUseMinDistance=true
    Scalability Mode=::NewEnumerator1
    Scale Spawn Count=true
    Spawn Count Scale By Distance=DataInterface() {propertyValues {Curve {keys [<at 0 1>,<at 0.95 1>,<at 1 0>]}}}
    UseLoopDelay=false
   )
   &SpawnRate(
    Spawn Probability=&FloatFromCurve(
     FloatCurve=DataInterface() {propertyValues {Curve {
        defaultValue=1
        keys [
         {interpMode=RCIM_Cubic, tangentMode=RCTM_User, time=0, value=1}
         {arriveTangent=-2, interpMode=RCIM_Cubic, leaveTangent=-0.3147, tangentMode=RCTM_Break, time=0.3, value=0}
         {interpMode=RCIM_Cubic, tangentMode=RCTM_User, time=0.7, value=0}
         {tangentMode=RCTM_User, time=1, value=1}
        ]
       }}}
     CurveIndex=Linked() {linkedVariable {name=NPC.FortniteNPC.TimeOfDay_0-1, type=NiagaraFloat}}
    )
    SpawnRate=&RandomRangeFloat(Minimum=50, Maximum=9)
    Use Spawn Probability=true
   )
   /SetVariables(Emitter.CameraPosition:Vector3f=&ReturnCameraPosition(Coordinate Space:ENiagaraCoordinateSpace=::Simulation, Camera Query:NiagaraDataInterfaceCamera=DataInterface())) {isSetParameters=true}
  ]
  particleSpawnScript [
   &InitializeParticle(
    Color=<rgba 3 3 3 1>
    Lifetime=&RandomRangeFloat(Minimum=8, Maximum=10)
    Mesh Scale Mode=::NewEnumerator3
    Mesh Uniform Scale=&Multiply_Float(A=&RandomRangeFloat(Minimum=0.15, Maximum=0.3), B=2)
    UsePositionOffset=false
    Write Material Random=true
   )
   &ShapeLocation(
    Enable Torus Knot Radius=false
    Handle Radius:NiagaraFloat=&Multiply_Float(A=300, B=$Emitter.RadiusMult)
    Large Radius=&Multiply_Float(A=&RandomRangeFloat(Minimum=1500, Maximum=6000), B=$Emitter.RadiusMult)
    MaskSpawnGroup=false
    Offset:NiagaraPosition=$Emitter.PositionOffset
    Offset Mode=::NewEnumerator0
    Override Seed=false
    Shape Primitive=::NewEnumerator3
    Torus Mode:ENiagaraShapeTorusMode=::NewEnumerator1
    UseOwnerScale=true
   )
   /SetVariables(Particles.KilledBYSpawnLimit?:NiagaraBool=&SetBoolByFloatComparison(
     A=&Divide_Float(A=&MakeFloatFromInt(Int=$Engine.Emitter.NumParticles), B=32)
     B=1
    )) {isSetParameters=true}
   &KillParticles(Kill Particles=&SetBoolByFloatComparison(
     A=&Divide_Float(A=&MakeFloatFromInt(Int=$Engine.Emitter.NumParticles), B=32)
     B=1
    )) {module=KillParticles001}
  ]
  particleUpdateScript [
   &ParticleState()
   /SetVariables(Particles.RandomDirection:Vector3f=&Lerp_Vector(
     A=&MakeVector(
      X=&Waveform(
       Global Phase=&Add_Float(A=$Particles.Age, B=&RandomRangeFloat(Minimum=0, Maximum=5))
       "Waveform [1]"=::NewEnumerator8
       "[1] Frequency"=1
       "[1] Amplitude Min/Max"=<-1 1>
       "[2] Phase"=$Emitter.Age
       "[3] Phase"=$Emitter.Age
      )
      Y=&Waveform(
       Global Phase=&Add_Float(A=$Particles.Age, B=&RandomRangeFloat(Minimum=0, Maximum=5))
       "Waveform [1]"=::NewEnumerator8
       "[1] Frequency"=1
       "[1] Amplitude Min/Max"=<-1 1>
       "[2] Phase"=$Emitter.Age
       "[3] Phase"=$Emitter.Age
      )
      Z=0
     )
     B=<1 0 0>
     Alpha=&Waveform(
      Global Phase=&Add_Float(A=$Particles.Age, B=&RandomRangeFloat(Minimum=0, Maximum=5))
      Global Clamp Min=0.4
      "Waveform [1]"=::NewEnumerator8
      "[1] Frequency"=1
      "[1] Amplitude Min/Max"=<0 1>
      "[2] Phase"=$Emitter.Age
      "[3] Phase"=$Emitter.Age
     )
    )) {enabled=false, isSetParameters=true}
   &InterpolateOverTime(
    Target Vector Value:Vector3f=&NormalizeVector(Vector To Normalize=$Particles.RandomDirection, Safe Normalize=true, Safe Fallback Threshold=0, Safe Fallback Vector=<0 0 0.001>)
    Variable Type=::NewEnumerator2
   ) {enabled=false, module=InterpolateOverTime001}
   &WindForce(
    Air Resistance=0.1
    Wind Speed=&TransformVector(Vector=$Particles.InterpolateOverTime001.InterpolateOverTime_V3_.Moving Average, Destination Space=::Simulation)
    Wind Speed Scale=&RandomRangeFloat(Minimum=100, Maximum=200)
   ) {enabled=false}
   &PointAttractionForce(
    Attraction Radius=10000
    Attraction Strength=&RandomRangeFloat(Minimum=3, Maximum=6)
    Attractor Position:Vector3f=$Particles.FishMoverPosition
    Kill Within Radius=false
    Module.Use Falloff=true
    UseAttractorOffset=false
   ) {enabled=false}
   &PointAttractionForce(
    Attraction Radius=500
    Attraction Strength=-200
    Attractor Position:Vector3f=$Emitter.CameraPosition
    Kill Within Radius=false
    Module.Use Falloff=true
    UseAttractorOffset=false
   ) {module=PointAttractionForce001}
   &VortexForce(
    Origin Pull Amount:NiagaraFloat=&RandomRangeFloat(Minimum=50, Maximum=120)
    Origin Pull Only On Vortex Axis=false
    UseOriginOffset=true
    Vortex Axis=<0 0 4>
    Vortex Force Amount=100
    Vortex Origin:NiagaraPosition=&AddVectorToPosition(Position=$Engine.Emitter.SimulationPosition, Vector=Linked() {linkedVariable {name=Emitter.PositionOffset, type=NiagaraPosition}})
    Vortex Origin Offset:Vector3f=&RandomRangeVector(Minimum=<0 0 222>, Maximum=<0 0 -222>)
   )
   &VortexForce(
    Origin Pull Amount:NiagaraFloat=&RandomRangeFloat(Minimum=1, Maximum=800)
    UseOriginOffset=false
    Vortex Force Amount=&FloatFromCurve(
     FloatCurve=DataInterface() {propertyValues {Curve {keys [
         {tangentMode=RCTM_User, time=0, value=1}
         {arriveTangent=-7.824, interpMode=RCIM_Cubic, leaveTangent=-7.824, tangentMode=RCTM_User, time=0.4, value=1}
         {interpMode=RCIM_Cubic, leaveTangent=-8.0089, time=1, value=0}
        ]}}}
     Scale Curve=222
    )
   ) {enabled=false, module=VortexForce001}
   /Niagara/Modules/Math/AvoidDistanceFieldSurfaces_GPU(Avoid Nearest Surface=true, Avoid Oncoming Surfaces=true, Nearest Surface Avoidance Strength=800, Obstactle Avoidance Strength=800)
   &Collision(
    Age Colliding Particles=false
    GPU Collision Type=::NewEnumerator2
    KillOnCollision=false
    Module.Control Rest State Via Surface Normal=false
    Module.EnableMaxCollisionCount=false
    Module.Randomize Collision Normal Vector=false
    Radius Calculation Type=::NewEnumerator1
    Restitution=1
   )
   &CurlNoiseForce(Bias Noise=false, Noise Frequency=19.1141, Noise Strength=995.4374, Pan Noise=false) {enabled=false}
   &CurlNoiseForce(
    Bias Noise=false
    Noise Frequency=&UniformAOrBFloat(
     A=0
     B=&RandomRangeFloat(Minimum=19, Maximum=5)
     Distribution Weight=0.15
     Evaluation Type=::NewEnumerator0
     Recalculate Random Each Loop=true
     Override Seed=false
    )
    Noise Strength=&UniformAOrBFloat(
     A=0
     B=&RandomRangeFloat(Minimum=555, Maximum=955)
     Distribution Weight=0.15
     Evaluation Type=::NewEnumerator0
     Recalculate Random Each Loop=true
     Override Seed=false
    )
    Pan Noise=true
    Pan Noise Field=<1 0 0>
   ) {module=CurlNoiseForce001}
   /SetVariables(Particles.RandomPosition:NiagaraPosition=&Lerp_Vector(
     A=Linked() {linkedVariable {name=Particles.ShapeLocation.ShapePosition, type=NiagaraPosition}}
     B=&MakeVector(X=0, Y=0, Z=0)
     Alpha=&FloatFromCurve(FloatCurve=DataInterface() {propertyValues {Curve {
         defaultValue=1
         keys [{interpMode=RCIM_Cubic, tangentMode=RCTM_User, time=0.75, value=1}, {interpMode=RCIM_Cubic, tangentMode=RCTM_User, time=0.8, value=0}]
        }}})
    )) {enabled=false, isSetParameters=true}
   &SolveForcesAndVelocity(Clamp Velocity=true, Limit Acceleration=false, Speed Limit=600)
   /SetVariables(Particles.New Vector 3f:Vector3f=&Multiply_VectorByFloat(
     Vector=$Particles.Scale
     Float=&FloatFromCurve(FloatCurve=DataInterface() {propertyValues {Curve {keys [
          {interpMode=RCIM_Cubic, tangentMode=RCTM_User, time=0, value=0}
          {interpMode=RCIM_Cubic, tangentMode=RCTM_User, time=0.1, value=1}
          {interpMode=RCIM_Cubic, tangentMode=RCTM_User, time=0.8, value=1}
          {tangentMode=RCTM_User, time=1, value=0}
         ]}}})
    )) {isSetParameters=true}
   &KillParticlesInVolume(Invert Volume=true, Sphere Radius=2000) {enabled=false}
   &DynamicMaterialParameters(
    Index 0 Param 1=0.25
    Index 0 Param 2=&RandomRangeFloat(Minimum=2.8, Maximum=3.5)
    Index 0 Param 3=0.25
    Index 0 Param 4=0.1
    Index 1 Param 1=1
    Index 1 Param 2=&RandomRangeFloat(Minimum=25, Maximum=55)
    Index 1 Param 3=2
    Index 1 Param 4=0.25
    Index 2 Param 1=1
    Index 2 Param 2=1
    Index 2 Param 3=2
    Index 2 Param 4=0.25
    Module.Index0Param0Write=true
    Module.Index0Param1Write=true
    Module.Index0Param2Write=true
    Module.Index0Param3Write=true
    Module.Index1Param0Write=true
    Module.Index1Param1Write=true
    Module.Index1Param2Write=true
    Module.Index1Param3Write=true
    Module.Index2Param0Write=true
    Module.Index2Param1Write=true
    Module.Index2Param2Write=true
    Module.Index2Param3Write=true
    Write Parameter Index 1=true
    Write Parameter Index 2=true
   ) {module=DynamicMaterialParameters001}
  ]
  rendererClasses [&NiagaraMeshRendererProperties]
  renderers [{rendererClass=&NiagaraMeshRendererProperties, rendererIndex=0}]
 }
 Fog {
  simTarget=GPUComputeSim
  emitterSpawnScript [/SetVariables(Emitter.ScaleColorAlpha=20, Emitter.ScaleMult:NiagaraFloat=&RandomRangeFloat(Minimum=2, Maximum=2.5)) {isSetParameters=true}]
  emitterUpdateScript [
   &EmitterState()
   &SpawnRate(
    SpawnRate=&FloatFromCurve(
     FloatCurve=DataInterface() {propertyValues {Curve {
        defaultValue=1
        keys [
         {interpMode=RCIM_Cubic, tangentMode=RCTM_User, time=0, value=1}
         {interpMode=RCIM_Cubic, tangentMode=RCTM_User, time=0.5, value=0}
         {tangentMode=RCTM_User, time=1, value=1}
        ]
       }}}
     CurveIndex=Linked() {linkedVariable {name=NPC.FortniteNPC.TimeOfDay_0-1, type=NiagaraFloat}}
     Scale Curve=12
    )
    Use Spawn Probability=false
   ) {enabled=false, module=SpawnRate001}
   &SpawnRate(SpawnRate=12, Use Spawn Probability=false)
  ]
  particleSpawnScript [
   &InitializeParticle(
    Color=<rgba 5.4947 7.558 16 0.25>
    Lifetime=&RandomRangeFloat(Minimum=4, Maximum=6)
    Mesh Scale Mode=::NewEnumerator3
    Mesh Uniform Scale=&Multiply_Float(A=&RandomRangeFloat(Minimum=4, Maximum=9), B=$Emitter.ScaleMult)
    Sprite Size Mode=::NewEnumerator3
    Uniform Sprite Size=&Multiply_Float(A=&RandomRangeFloat(Minimum=500, Maximum=750), B=$Emitter.ScaleMult)
    UsePositionOffset=false
    Write Material Random=true
   ) {module=InitializeParticle002}
   &ShapeLocation(MaskSpawnGroup=false, Offset Mode=::NewEnumerator0, Override Seed=false, Sphere Radius=8500, UseOwnerScale=true) {module=ShapeLocation001}
   &self:SampleLandscape_02() {enabled=false, module=SampleLandscape_02}
   &KillParticlesInVolume(Invert Volume=true, Sphere Radius=6000) {enabled=false, module=KillParticlesInVolume001}
   &KillParticles()
   /Game/Effects/Niagara/Modules/Spawn/FN_Spawn_GBuffer_Color_WithinRangeV2(
    HueMax:NiagaraFloat=&UniformAOrBFloat(A=35, B=95, Evaluation Type=::NewEnumerator0, Recalculate Random Each Loop=true, Override Seed=false)
    HueMin:NiagaraFloat=&UniformAOrBFloat(A=30, B=72, Evaluation Type=::NewEnumerator0, Recalculate Random Each Loop=true, Override Seed=false)
   ) {enabled=false}
   &FN_Spawn_GBuffer_Color_WithinRangeV2(HueMax=95, HueMin=72) {enabled=false, module=FN_Spawn_GBuffer_Color_WithinRangeV2001}
   &Color(Color=&RandomRangeLinearColor(Minimum=<rgba 0.305 0.305 0.2462 1>, Maximum=<rgba 0.2051 0.1878 0.1559 1>)) {enabled=false}
   &Color(Color=&RandomRangeLinearColor(Minimum=<rgba 0.0865 0.0648 0.0343 1>, Maximum=<rgba 0.1529 0.0908 0.0319 1>)) {enabled=false, module=Color001}
  ]
  particleUpdateScript [
   &ParticleState()
   &self:SampleLandscape_02() {module=SampleLandscape_02001}
   /SetVariables(Particles.Position=&AddVectorToPosition(
     Position=$Particles.Position
     Vector=&Multiply_Vector(A=$Output.SampleLandscape_02001.LandscapeNormal, B=<1 1 200>)
    )) {isSetParameters=true}
   &KillParticlesInVolume(
    Box Size:Vector3f=<3000 7000 3000>
    Kill Shape=::NewEnumerator1
    Volume Origin:Vector3f=&AddVectorToPosition(Position=<-79152 -29616 5540>, Vector=<0 0 0>)
   ) {enabled=false}
   &KillParticlesInVolume(
    Box Size:Vector3f=<6000 7000 4000>
    Kill Shape=::NewEnumerator1
    Volume Origin:Vector3f=&AddVectorToPosition(Position=<-80932 -29612 5172>, Vector=<0 0 0>)
   ) {module=KillParticlesInVolume002}
   &KillParticlesInVolume(
    Box Size:Vector3f=<5000 4000 4000>
    Kill Shape=::NewEnumerator1
    Volume Origin:Vector3f=&AddVectorToPosition(Position=<-66204 -28636 4608>, Vector=<0 0 0>)
   ) {module=KillParticlesInVolume003}
   &ScaleMeshSize(Scale Factor=&Lerp_Vector(
     A=<0.125 0.125 0.125>
     B=<1 1 1>
     Alpha=&FN_LOD_Distance_Remap(Start Distance=0, End Distance=3500, Minimum Output=0, Maximum Output=1, Clamp Output=true)
    ))
   &ScaleSpriteSize(
    Uniform Curve Index=&FN_LOD_Distance_Remap(Start Distance=0, End Distance=1500, Minimum Output=0, Maximum Output=1, Clamp Output=true)
    Uniform Curve Sprite Scale=DataInterface() {propertyValues {Curve {keys [{interpMode=RCIM_Cubic, tangentMode=RCTM_Break, time=0, value=0}, {tangentMode=RCTM_User, time=1, value=1}]}}}
   )
   &ScaleColor(
    Scale Alpha=&Multiply_Float(
     A=&Multiply_Float(
      A=&FloatFromCurve(
       FloatCurve=DataInterface() {propertyValues {Curve {keys [
           {interpMode=RCIM_Cubic, tangentMode=RCTM_User, time=0, value=0}
           {interpMode=RCIM_Cubic, tangentMode=RCTM_User, time=0.5, value=1}
           {tangentMode=RCTM_User, time=1, value=0}
          ]}}}
       Scale Curve=&Multiply_Float(
        A=&Multiply_Float(A=&RandomRangeFloat(Minimum=0.15, Maximum=0.17), B=0.7)
        B=$Emitter.ScaleColorAlpha
       )
      )
      B=&FloatFromCurve(
       FloatCurve=DataInterface() {propertyValues {Curve {
          defaultValue=1
          keys [
           {interpMode=RCIM_Cubic, tangentMode=RCTM_User, time=0, value=1.2}
           {arriveTangent=2.3353, interpMode=RCIM_Cubic, leaveTangent=-0.0832, tangentMode=RCTM_Break, time=0.25, value=1.5}
           {interpMode=RCIM_Cubic, tangentMode=RCTM_User, time=0.5, value=1.2}
           <at 0.75 1 RCIM_Cubic>
           {tangentMode=RCTM_User, time=1, value=1.2}
          ]
         }}}
       CurveIndex=Linked() {linkedVariable {name=NPC.FortniteNPC.TimeOfDay_0-1, type=NiagaraFloat}}
      )
     )
     B=&FloatFromCurve(
      FloatCurve=DataInterface() {propertyValues {Curve {
         defaultValue=0
         keys [
          {arriveTangent=1.6813, interpMode=RCIM_Cubic, leaveTangent=1.6813, tangentMode=RCTM_Break, time=0, value=0.05}
          {interpMode=RCIM_Cubic, tangentMode=RCTM_User, time=1, value=0.6}
         ]
        }}}
      CurveIndex=&RemapRange(
       Input Value To Remap=&VectorLength(Vector=&Subtract_Position(A=$Engine.Owner.Position, B=$Particles.Position))
       Input Min=0
       Input Max=10000
       Output Min=0
       Output Max=0.08
      )
      Scale Curve=&RandomRangeFloat(Minimum=0.9, Maximum=1)
     )
    )
    ScaleA=true
    ScaleRGB=false
   ) {enabled=false, module=ScaleColor001}
   &ScaleColor(
    Scale Alpha=&FloatFromCurve(
     FloatCurve=DataInterface() {propertyValues {Curve {keys [
         {interpMode=RCIM_Cubic, tangentMode=RCTM_User, time=0, value=0}
         {interpMode=RCIM_Cubic, tangentMode=RCTM_User, time=0.5, value=1}
         {tangentMode=RCTM_User, time=1, value=0}
        ]}}}
     Scale Curve=0.1
    )
    Scale RGB=&VectorFromFloat(Value=&FloatFromCurve(
      FloatCurve=DataInterface() {propertyValues {Curve {
         defaultValue=1
         keys [
          {interpMode=RCIM_Cubic, tangentMode=RCTM_User, time=0.3, value=1}
          {interpMode=RCIM_Cubic, tangentMode=RCTM_User, time=0.5, value=0.009}
          {tangentMode=RCTM_User, time=0.7, value=1}
         ]
        }}}
      CurveIndex=Linked() {linkedVariable {name=NPC.FortniteNPC.TimeOfDay_0-1, type=NiagaraFloat}}
     ))
    ScaleA=true
    ScaleRGB=true
   )
   &AccelerationForce(Acceleration=<150 150 25>)
   &Drag(Use Linear Drag=true, Use Rotational Drag=true)
   &SolveForcesAndVelocity(Clamp Velocity=false, Limit Acceleration=false)
   &CameraOffset(Camera Offset Amount=20) {enabled=false}
   &CameraOffset(Camera Offset Amount=90) {module=CameraOffset001}
  ]
  rendererClasses [&NiagaraSpriteRendererProperties, &NiagaraMeshRendererProperties]
  renderers [
   {rendererClass=&NiagaraSpriteRendererProperties}
   {rendererClass=&NiagaraMeshRendererProperties, rendererIndex=1}
   {rendererClass=&NiagaraSpriteRendererProperties, rendererIndex=2}
  ]
 }
}
summary {
 systemName=NS_FreakyFields_Atmosphere
 emitters [
  {
   emitterName=Fog
   simTarget=GPUComputeSim
   rendererClasses [&NiagaraSpriteRendererProperties, &NiagaraMeshRendererProperties]
  }
  {emitterName=Bats001, simTarget=GPUComputeSim, rendererClasses [&NiagaraMeshRendererProperties]}
 ]
 userVariables [User.Radius Mult {type=NiagaraFloat, defaultValue=1, description=""}]
}
